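# Conversation database for the Baba Yaga NPC, defining the greeting shown when a player starts talking to her.
# A single entry point feeds a speech node with a default line and contextual variants for returning after time away, holding bars, or a first-ever conversation.
# Each variant is gated by a controller requirement and ranked by a priority curve row.
# The speech node then links out to the shared NPC quest and basic services conversations.
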ConversationDatabase /NPCLibrary/NPCs/BabaYaga/BabaYaga/NPCLibrary_Conversation_BabaYaga
entry_points [{tag=Conversation.Entry.BabaYaga, destinations [#1]}]
nodes [
 ConversationEntryPointNode_7 {
  id=#1
  class=ConversationEntryPointNode
  outputs [#2]
  properties {entryTag {tagName=Conversation.Entry.BabaYaga}}
 }
 Speech_C_4 {
  id=#2
  class=Speech_C
  outputs [#3, #4]
  properties {
   nodeName=""
   generalConfig {
    defaultMessage="The trees told me you'd be coming. Ha!"
    contextualMessages [
     {
      message="Ha! Your return has long been written in the stars."
      contextRequirements [{requirement {class=FortControllerRequirement_CheckTimeSinceLastPlayedMatch}}]
      priority {curve {rowName=Tandem.Speech.Priority.WhileSincePlayed}}
     }
     {
      message="Bars? Meh. Teeth, on the other hand... Ha!"
      contextRequirements [{requirement {class=FortControllerRequirement_HasBars}}]
      priority {curve {rowName=Tandem.Speech.Priority.HighTotalBars}}
     }
     {
      message="Ha! You've got a shine to you... the promise of something to pass."
      contextRequirements [{requirement {class=FortControllerRequirement_IsFirstEverConversationWithNPC}}]
      priority {curve {rowName=Tandem.Speech.Priority.FirstEverInteraction}}
     }
    ]
   }
  }
 }
 ConversationLinkNode_0 {id=#4, class=ConversationLinkNode, properties {remoteEntryTag {tagName=Conversation.Entry.BasicServices}}}
 ConversationLinkNode_1 {id=#3, class=ConversationLinkNode, properties {remoteEntryTag {tagName=Conversation.Entry.NPCLibrary.Quests}}}
]
